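# Sound cue for a TNT bounce effect, mixing an explode-bounce layer with a flutter layer.
# Each layer picks an equally weighted random wave and varies its pitch slightly; the bounce layer also varies volume and is mixed quieter than the flutter.
# It uses grenade-bounce attenuation and the ranged-weapon projectile effects sound class, with a concurrency override that blocks new instances once its small limit is reached.
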SoundCue /Game/Sounds/Fort_GamePlay_Sounds/Explosives/TnT/TNT_Explode_Bounce_Cue
root=Mixer_0
nodes [
 {id=Mixer_0, children [Modulator_5,Modulator_0], properties {inputVolume [0.5,1]}}
 {id=Modulator_0, children [Random_1], properties {pitchMax=1.1, pitchMin=0.9, volumeMax=1, volumeMin=1}}
 {id=Modulator_5, children [Random_0], properties {pitchMax=1.1, pitchMin=0.9, volumeMax=1, volumeMin=0.5}}
 {
  id=Random_0
  children [WavePlayer_0,WavePlayer_1,WavePlayer_2,WavePlayer_3,WavePlayer_4,WavePlayer_5]
  properties {bShouldExcludeFromBranchCulling=true, weights [1,1,1,1,1,1]}
 }
 {
  id=Random_1
  children [WavePlayer_6, WavePlayer_7, WavePlayer_8, WavePlayer_9, WavePlayer_10]
  properties {bShouldExcludeFromBranchCulling=true, weights [1,1,1,1,1]}
 }
 {id=WavePlayer_0, properties {soundWaveAssetPtr=&/TNT_Explode_Bounce_01}}
 {id=WavePlayer_1, properties {soundWaveAssetPtr=&/TNT_Explode_Bounce_02}}
 {id=WavePlayer_10, properties {soundWaveAssetPtr=&/TNT_Bounce_Flutter_01}}
 {id=WavePlayer_2, properties {soundWaveAssetPtr=&/TNT_Explode_Bounce_03}}
 {id=WavePlayer_3, properties {soundWaveAssetPtr=&/TNT_Explode_Bounce_04}}
 {id=WavePlayer_4, properties {soundWaveAssetPtr=&/TNT_Explode_Bounce_05}}
 {id=WavePlayer_5, properties {soundWaveAssetPtr=&/TNT_Explode_Bounce_06}}
 {id=WavePlayer_6, properties {soundWaveAssetPtr=&/TNT_Bounce_Flutter_05}}
 {id=WavePlayer_7, properties {soundWaveAssetPtr=&/TNT_Bounce_Flutter_04}}
 {id=WavePlayer_8, properties {soundWaveAssetPtr=&/TNT_Bounce_Flutter_03}}
 {id=WavePlayer_9, properties {soundWaveAssetPtr=&/TNT_Bounce_Flutter_02}}
]
properties {
 attenuationSettings=/Game/Sounds/Attenuation/Fort_Grenade_Bounce_Att
 bOverrideConcurrency=true
 maxDistance=5500
 soundClassObject=&Fort_WeaponRanged_Projectile_Effects
 volumeMultiplier=1
 concurrencyOverrides {maxCount=3, resolutionRule=PreventNew}
}
