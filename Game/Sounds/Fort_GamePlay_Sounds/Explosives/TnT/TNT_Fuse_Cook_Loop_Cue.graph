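# Sound cue for the TNT fuse cook loop, layering an ignite one-shot with a looping fuse burn.
# A mixer plays one of four randomly picked ignite waves alongside the looping fuse wave.
# The loop runs through an enveloper that fades its volume in along a curve, then a modulator that randomizes pitch.
# It uses the explosive fuse cook attenuation and Explosives sound class, with overridden concurrency capping instances.
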SoundCue /Game/Sounds/Fort_GamePlay_Sounds/Explosives/TnT/TNT_Fuse_Cook_Loop_Cue
root=Mixer_0
nodes [
 {
  id=Enveloper_0
  children [WavePlayer_1]
  properties {
   pitchCurve {editorCurveData {keys [{time=0}]}}
   volumeCurve {editorCurveData {keys [{arriveTangent=0.9158, interpMode=RCIM_Cubic, leaveTangent=0.9158, tangentMode=RCTM_User, time=0, value=0}, <at 1 1>]}}
  }
 }
 {id=Mixer_0, children [Modulator_1,Random_0], properties {inputVolume [1,1]}}
 {id=Modulator_1, children [Enveloper_0], properties {pitchMax=1.2, pitchMin=0.9, volumeMax=1, volumeMin=1}}
 {id=Random_0, children [WavePlayer_2,WavePlayer_3,WavePlayer_4,WavePlayer_5], properties {weights [1,1,1,1]}}
 {id=WavePlayer_1, properties {bLooping=true, soundWaveAssetPtr=&/TNT_Fuse_Loop01}}
 {id=WavePlayer_2, properties {soundWaveAssetPtr=&/TNT_Ignite_01}}
 {id=WavePlayer_3, properties {soundWaveAssetPtr=&/TNT_Ignite_02}}
 {id=WavePlayer_4, properties {soundWaveAssetPtr=&/TNT_Ignite_03}}
 {id=WavePlayer_5, properties {soundWaveAssetPtr=&/TNT_Ignite_04}}
]
properties {
 attenuationSettings=/Game/Sounds/Attenuation/Fort_Explosive_Fuse_Prj_Cook_Att
 bOverrideConcurrency=true
 maxDistance=6300
 soundClassObject=/Game/Sounds/FortSoundClassesAndModes/Fort_Explosives
 volumeMultiplier=0.5
 attenuationOverrides {distanceAlgorithm=NaturalSound, falloffDistance=2000, stereoSpread=0}
 concurrencyOverrides {maxCount=3}
}
