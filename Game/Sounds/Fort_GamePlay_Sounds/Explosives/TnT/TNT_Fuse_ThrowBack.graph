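# Sound cue for throwing back TNT with a lit fuse, routed to the explosives sound class with its own concurrency limit.
# A mixer layers a random impulse-grenade drop, a pitch- and volume-modulated TNT fuse ignite, and a random frag-grenade throw that is pitch-modulated and slightly delayed.
# The ignite and throw layers each pass through their own attenuation node, while the drop layer feeds the mixer directly.
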SoundCue /Game/Sounds/Fort_GamePlay_Sounds/Explosives/TnT/TNT_Fuse_ThrowBack
root=Mixer_0
nodes [
 {id=Attenuation_1, children [Modulator_0], properties {attenuationSettings=/Game/Sounds/Attenuation/Fort_Explosive_Fuse_Ambient_Att}}
 {id=Attenuation_2, children [Delay_0], properties {attenuationSettings=/Game/Sounds/Attenuation/Fort_Explosive_Fuse_Prj_Throw_Att}}
 {id=Delay_0, children [Modulator_2], properties {delayMax=0.2, delayMin=0.2}}
 {id=Mixer_0, children [Attenuation_2, Random_1, Attenuation_1], properties {inputVolume [0.5,0.7,0.6]}}
 {id=Modulator_0, children [WavePlayer_1], properties {pitchMax=1.2, volumeMax=1, volumeMin=0.8}}
 {id=Modulator_2, children [Random_0], properties {pitchMax=1.3}}
 {id=Random_0, children [WavePlayer_6,WavePlayer_7,WavePlayer_8], properties {bShouldExcludeFromBranchCulling=true, weights [1,1,1]}}
 {id=Random_1, children [WavePlayer_0,WavePlayer_2,WavePlayer_3], properties {weights [1,1,1]}}
 {id=WavePlayer_0, properties {soundWaveAssetPtr=/Game/Sounds/Fort_GamePlay_Sounds/Explosives/ShockwaveGrenade/ImpulseGrenade_Drop_01}}
 {id=WavePlayer_1, properties {soundWaveAssetPtr=&/TNT_Fuse_Ignite_01}}
 {id=WavePlayer_2, properties {soundWaveAssetPtr=/Game/Sounds/Fort_GamePlay_Sounds/Explosives/ShockwaveGrenade/ImpulseGrenade_Drop_02}}
 {id=WavePlayer_3, properties {soundWaveAssetPtr=/Game/Sounds/Fort_GamePlay_Sounds/Explosives/ShockwaveGrenade/ImpulseGrenade_Drop_03}}
 {id=WavePlayer_6, properties {soundWaveAssetPtr=/Game/Sounds/Fort_Weapons/Grenades/Frag/Grenade_Frag_Throw_01}}
 {id=WavePlayer_7, properties {soundWaveAssetPtr=/Game/Sounds/Fort_Weapons/Grenades/Frag/Grenade_Frag_Throw_02}}
 {id=WavePlayer_8, properties {soundWaveAssetPtr=/Game/Sounds/Fort_Weapons/Grenades/Frag/Grenade_Frag_Throw_03}}
]
properties {
 bOverrideConcurrency=true
 maxDistance=13000
 soundClassObject=/Game/Sounds/FortSoundClassesAndModes/Fort_Explosives
 volumeMultiplier=0.6
 attenuationOverrides {distanceAlgorithm=NaturalSound, falloffDistance=2000, stereoSpread=0}
 concurrencyOverrides {maxCount=3}
}
