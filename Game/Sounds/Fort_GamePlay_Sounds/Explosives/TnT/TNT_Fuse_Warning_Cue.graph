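# Sound cue for the TNT fuse warning in the explosives sound class, played at reduced volume.
# One of three equally weighted warning waves is picked at random and passed through a modulator with pitch variation flattened.
# An enveloper then ramps volume and pitch upward over the first seconds, so the warning grows more urgent.
# It uses the shared explosive fuse attenuation asset and overrides concurrency to cap simultaneous instances at two.
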SoundCue /Game/Sounds/Fort_GamePlay_Sounds/Explosives/TnT/TNT_Fuse_Warning_Cue
root=Enveloper_0
nodes [
 {
  id=Enveloper_0
  children [Modulator_0]
  properties {pitchCurve {editorCurveData {keys [<at 0 0.9>,<at 2 1.5>]}}, volumeCurve {editorCurveData {keys [<at 0 0.5>,<at 1 1>]}}}
 }
 {id=Modulator_0, children [Random_1], properties {pitchMax=1, pitchMin=1}}
 {id=Random_1, children [WavePlayer_5,WavePlayer_0,WavePlayer_6], properties {bShouldExcludeFromBranchCulling=true, weights [1,1,1]}}
 {id=WavePlayer_0, properties {soundWaveAssetPtr=&/TNT_Fuse_Warning01}}
 {id=WavePlayer_5, properties {soundWaveAssetPtr=&/TNT_Fuse_Warning02}}
 {id=WavePlayer_6, properties {soundWaveAssetPtr=&/TNT_Fuse_Warning03}}
]
properties {
 attenuationSettings=/Game/Sounds/Attenuation/Fort_Explosive_Fuse_Ambient_Att
 bOverrideConcurrency=true
 maxDistance=3500
 soundClassObject=/Game/Sounds/FortSoundClassesAndModes/Fort_Explosives
 volumeMultiplier=0.5
 attenuationOverrides {distanceAlgorithm=NaturalSound, falloffDistance=2000, stereoSpread=0}
 concurrencyOverrides {maxCount=2}
}
